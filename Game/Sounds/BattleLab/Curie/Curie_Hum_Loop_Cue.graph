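# Ambient hum cue for Curie in BattleLab, assigned to the Athena ambient sound class.
# A Mixer layers two looping platform waves with a much quieter pulse wave, which is not flagged as looping.
# A Modulator at the root applies slight random pitch variation to the mix.
# Attenuation and concurrency are overridden on the cue, adding short-range HRTF spatialization, reverb send and a capped voice count.
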SoundCue /Game/Sounds/BattleLab/Curie/Curie_Hum_Loop_Cue
root=Modulator_1
nodes [
 {id=Mixer_1, children [WavePlayer_11, WavePlayer_0, WavePlayer_8], properties {inputVolume [0.9,0.9,0.2]}}
 {id=Modulator_1, children [Mixer_1], properties {pitchMax=1.02, pitchMin=0.98}}
 {id=WavePlayer_0, properties {bLooping=true, soundWaveAssetPtr=&/Curie_PlatformLoop01}}
 {id=WavePlayer_11, properties {bLooping=true, soundWaveAssetPtr=&/Curie_PlatformLoop02}}
 {id=WavePlayer_8, properties {soundWaveAssetPtr=&/Curie_PulseLoop03}}
]
properties {
 bOverrideAttenuation=true
 bOverrideConcurrency=true
 maxDistance=1512
 soundClassObject=&Fort_Ambient_Athena
 volumeMultiplier=0.416
 attenuationOverrides {
  falloffDistance=1000
  reverbWetLevelMax=0.5
  reverbWetLevelMin=0.1
  spatializationAlgorithm=SPATIALIZATION_HRTF
  stereoSpread=500
  attenuationShapeExtents {x=512}
 }
 concurrencyOverrides {maxCount=6, volumeScale=0.7}
}
