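# Sound cue for the Curie pulse sound in BattleLab, routed to the Athena ambient sound class.
# A Random node picks one of nine equally weighted pulse wave variants and feeds it through a Modulator that slightly varies pitch and volume on each play.
# It references a dedicated attenuation asset for spatial falloff and overrides concurrency to cap simultaneous instances.
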SoundCue /Game/Sounds/BattleLab/Curie/Curie_Pulse_Cue
root=Modulator_2
nodes [
 {id=Modulator_2, children [Random_0], properties {pitchMax=1.2, pitchMin=0.9, volumeMax=1.1, volumeMin=0.9}}
 {
  id=Random_0
  children [WavePlayer_0, WavePlayer_2, WavePlayer_3, WavePlayer_4, WavePlayer_5, WavePlayer_6, WavePlayer_7, WavePlayer_8, WavePlayer_10]
  properties {weights [1,1,1,1,1,1,1,1,1]}
 }
 {id=WavePlayer_0, properties {soundWaveAssetPtr=&/Curie_Pulse_04}}
 {id=WavePlayer_10, properties {soundWaveAssetPtr=&/Curie_Pulse_09}}
 {id=WavePlayer_2, properties {soundWaveAssetPtr=&/Curie_Pulse_08}}
 {id=WavePlayer_3, properties {soundWaveAssetPtr=&/Curie_Pulse_01}}
 {id=WavePlayer_4, properties {soundWaveAssetPtr=&/Curie_Pulse_02}}
 {id=WavePlayer_5, properties {soundWaveAssetPtr=&/Curie_Pulse_03}}
 {id=WavePlayer_6, properties {soundWaveAssetPtr=&/Curie_Pulse_05}}
 {id=WavePlayer_7, properties {soundWaveAssetPtr=&/Curie_Pulse_06}}
 {id=WavePlayer_8, properties {soundWaveAssetPtr=&/Curie_Pulse_07}}
]
properties {
 attenuationSettings=&/Curie_Pulse_Att
 bOverrideConcurrency=true
 maxDistance=1700
 soundClassObject=&Fort_Ambient_Athena
 volumeMultiplier=0.832
 attenuationOverrides {falloffDistance=1300, reverbWetLevelMax=0.5, reverbWetLevelMin=0.1, stereoSpread=500, attenuationShapeExtents {x=512}}
 concurrencyOverrides {maxCount=4}
}
